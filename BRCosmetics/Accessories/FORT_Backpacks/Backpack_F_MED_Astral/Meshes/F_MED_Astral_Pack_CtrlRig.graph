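# Control Rig for the Astral backpack cosmetic that procedurally animates a small creature with body, head, ears, teeth, arms, legs and feet.
# Chain harmonics drive the body with wave motion only, and each arm and leg chain with wave, pendulum and reach.
# Timeline sine waves at seeded random speeds, remapped into slerp factors, then swing the head, feet, teeth and ears.
# The head blends between fixed rotations, while the others offset their current local rotation.
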ControlRigBlueprint /BRCosmetics/Accessories/FORT_Backpacks/Backpack_F_MED_Astral/Meshes/F_MED_Astral_Pack_CtrlRig
graphs [
 RigVMFunctionLibrary {links [], nodes []}
 RigVMModel {nodes [
   RigUnit_BeginExecution = &RigUnit_BeginExecution()
   /Script/ControlRig.RigUnit_ChainHarmonicsPerItem(
    ChainRoot=(Type=Bone,Name="dyn_body_01")
    DrawWorldOffset=(Rotation=(X=0,Y=0,Z=0,W=1),Translation=(X=0,Y=0,Z=0),Scale3D=(X=1,Y=1,Z=1))
    Pendulum=(bEnabled=false,PendulumStiffness=2,PendulumGravity=(X=0,Y=0,Z=0),PendulumBlend=0.75,PendulumDrag=0.98,PendulumMaximum=0.1,UnwindAxis=(X=0,Y=1,Z=0),UnwindMinimum=0.2,UnwindMaximum=0.05)
    Reach=(bEnabled=false,ReachTarget=(X=0,Y=0,Z=0),ReachAxis=(X=1,Y=0,Z=0))
    Speed=(X=1,Y=1.2,Z=1)
    Wave=(bEnabled=True,WaveFrequency=(X=1,Y=0.6,Z=0.8),WaveAmplitude=(X=0.3,Y=0.3,Z=0.25),WaveOffset=(X=0,Y=0,Z=0),WaveNoise=(X=0,Y=0,Z=0),WaveMaximum=1)
    WaveCurve=(EditorCurveData=(Keys=((Time=-0.0046,Value=0.9948),(Time=0.8689,Value=0.4555))))
    bDrawDebug=false
    ExecutePin=@RigUnit_BeginExecution.ExecutePin
   ) {is=struct RigVMUnitNode}
   > &RigUnit_ChainHarmonicsPerItem(
    ChainRoot=(Type=Bone,Name="dyn_l_arm_01")
    DrawWorldOffset=(Rotation=(X=0,Y=0,Z=0,W=1),Translation=(X=0,Y=0,Z=0),Scale3D=(X=1,Y=1,Z=1))
    Pendulum=(bEnabled=True,PendulumStiffness=2,PendulumGravity=(X=0,Y=0,Z=0),PendulumBlend=0.75,PendulumDrag=0.98,PendulumMaximum=0.1,UnwindAxis=(X=0,Y=1,Z=0),UnwindMinimum=0.2,UnwindMaximum=0.05)
    Reach=(bEnabled=True,ReachTarget=(X=0,Y=0,Z=0),ReachAxis=(X=1,Y=0,Z=0))
    Speed=(X=1,Y=3,Z=1)
    Wave=(bEnabled=True,WaveFrequency=(X=1,Y=0.6,Z=0.8),WaveAmplitude=(X=0,Y=1,Z=0),WaveOffset=(X=0,Y=1,Z=2),WaveNoise=(X=0,Y=1,Z=0),WaveMaximum=1)
    WaveCurve=(EditorCurveData=(Keys=((Time=0.002,Value=0.8593),(Time=0.9954,Value=0.2388))))
    bDrawDebug=false
   )
   > &RigUnit_ChainHarmonicsPerItem(
    ChainRoot=(Type=Bone,Name="dyn_r_arm_01")
    DrawWorldOffset=(Rotation=(X=0,Y=0,Z=0,W=1),Translation=(X=0,Y=0,Z=0),Scale3D=(X=1,Y=1,Z=1))
    Pendulum=(bEnabled=True,PendulumStiffness=2,PendulumGravity=(X=0,Y=0,Z=0),PendulumBlend=0.75,PendulumDrag=0.98,PendulumMaximum=0.1,UnwindAxis=(X=0,Y=1,Z=0),UnwindMinimum=0.2,UnwindMaximum=0.05)
    Reach=(bEnabled=True,ReachTarget=(X=0,Y=0,Z=0),ReachAxis=(X=1,Y=0,Z=0))
    Speed=(X=1,Y=3,Z=1)
    Wave=(bEnabled=True,WaveFrequency=(X=1,Y=0.6,Z=0.8),WaveAmplitude=(X=0,Y=1,Z=0),WaveOffset=(X=0,Y=5,Z=0),WaveNoise=(X=0,Y=1,Z=0),WaveMaximum=1)
    WaveCurve=(EditorCurveData=(Keys=((Time=-0.0132,Value=0.8212),(Time=0.9954,Value=0.217))))
    bDrawDebug=false
   )
   > &RigUnit_ChainHarmonicsPerItem(
    ChainRoot=(Type=Bone,Name="dyn_l_leg_01")
    DrawWorldOffset=(Rotation=(X=0,Y=0,Z=0,W=1),Translation=(X=0,Y=0,Z=0),Scale3D=(X=1,Y=1,Z=1))
    Pendulum=(bEnabled=True,PendulumStiffness=2,PendulumGravity=(X=0,Y=0,Z=0),PendulumBlend=0.75,PendulumDrag=0.98,PendulumMaximum=0.1,UnwindAxis=(X=0,Y=1,Z=0),UnwindMinimum=0.2,UnwindMaximum=0.05)
    Reach=(bEnabled=True,ReachTarget=(X=0,Y=0,Z=0),ReachAxis=(X=1,Y=0,Z=0))
    Speed=(X=1,Y=2,Z=1)
    Wave=(bEnabled=True,WaveFrequency=(X=1,Y=0.6,Z=0.8),WaveAmplitude=(X=0,Y=0.3,Z=0),WaveOffset=(X=0,Y=0,Z=0),WaveNoise=(X=0,Y=1,Z=0),WaveMinimum=0,WaveMaximum=1)
    WaveCurve=(EditorCurveData=(Keys=((Time=0.0027,Value=0.0103),(Time=1,Value=1))))
    bDrawDebug=false
   )
   > &RigUnit_ChainHarmonicsPerItem(
    ChainRoot=(Type=Bone,Name="dyn_r_leg_01")
    DrawWorldOffset=(Rotation=(X=0,Y=0,Z=0,W=1),Translation=(X=0,Y=0,Z=0),Scale3D=(X=1,Y=1,Z=1))
    Pendulum=(bEnabled=True,PendulumStiffness=2,PendulumGravity=(X=0,Y=0,Z=0),PendulumBlend=0.75,PendulumDrag=0.98,PendulumMaximum=0.1,UnwindAxis=(X=0,Y=1,Z=0),UnwindMinimum=0.2,UnwindMaximum=0.05)
    Reach=(bEnabled=True,ReachTarget=(X=0,Y=0,Z=0),ReachAxis=(X=1,Y=0,Z=0))
    Speed=(X=1,Y=2,Z=1)
    Wave=(bEnabled=True,WaveFrequency=(X=1,Y=0.6,Z=0.8),WaveAmplitude=(X=0,Y=0.3,Z=0),WaveOffset=(X=0,Y=4,Z=0),WaveNoise=(X=0,Y=1,Z=0),WaveMinimum=0,WaveMaximum=1)
    WaveCurve=(EditorCurveData=(Keys=((Time=0.0027,Value=0.0103),(Time=1,Value=1))))
    bDrawDebug=false
   )
   n1 = /Script/RigVM.RigVMFunction_RandomFloat(Duration=1, Maximum=3, Minimum=2, Seed=100) {is=struct RigVMUnitNode, template="Random::Execute(in Seed,in Minimum,in Maximum,in Duration,out Result)"}
   n2 = &RigVMFunction_MathFloatRemap(
    SourceMaximum=1
    SourceMinimum=-1
    Value=/Script/RigVM.RigVMFunction_MathFloatSin(Value=/Script/RigVM.RigVMFunction_Timeline(Speed=@n1.Result) {is=struct RigVMUnitNode}.Time) {is=struct RigVMUnitNode, template="Sin::Execute(in Value,out Result)"}.Result
   )
   > &RigUnit_SetRotation(
    Item=(Type=Bone,Name="dyn_head")
    Weight=@n2.Result
    Value=/Script/RigVM.RigVMFunction_MathQuaternionSlerp(A=(X=-0.0262,Y=-0.0436,Z=0.9987,W=-0.0011), B=(X=0.0436,Y=0,Z=0.999,W=0), T=@n2.Result) {is=struct RigVMUnitNode, template="Interpolate::Execute(in A,in B,in T,out Result)"}.Result
   )
   n3 = &RigVMFunction_MathFloatSin(Value=&RigVMFunction_Timeline(Speed=&RigVMFunction_RandomFloat(Duration=1, Maximum=6, Minimum=4, Seed=100) {template="Random::Execute(in Seed,in Minimum,in Maximum,in Duration,out Result)"}.Result).Time) {template="Sin::Execute(in Value,out Result)"}
   n4 = &RigUnit_GetTransform(Item=(Type=Bone,Name="dyn_l_foot"), Space=LocalSpace)
   > &RigUnit_SetRotation(
    Item=(Type=Bone,Name="dyn_l_foot")
    Weight=1
    Value=&RigVMFunction_MathQuaternionSlerp(
     T=&RigVMFunction_MathFloatRemap(SourceMaximum=1, SourceMinimum=-1, Value=@n3.Result).Result
     B=/Script/RigVM.RigVMFunction_MathQuaternionMul(B=(X=0,Y=0,Z=0.1305,W=0.9914), A=@n4.Transform.Rotation) {is=struct RigVMUnitNode, template="Multiply::Execute(in A,in B,out Result)"}.Result
     A=@n4.Transform.Rotation
    ) {template="Interpolate::Execute(in A,in B,in T,out Result)"}.Result
   )
   n5 = &RigVMFunction_MathFloatSin(Value=&RigVMFunction_Timeline(Speed=&RigVMFunction_RandomFloat(Duration=1, Maximum=6, Minimum=4, Seed=100) {template="Random::Execute(in Seed,in Minimum,in Maximum,in Duration,out Result)"}.Result).Time) {template="Sin::Execute(in Value,out Result)"}
   n6 = &RigUnit_GetTransform(Item=(Type=Bone,Name="dyn_r_foot"), Space=LocalSpace)
   > &RigUnit_SetRotation(
    Item=(Type=Bone,Name="dyn_r_foot")
    Weight=1
    Value=&RigVMFunction_MathQuaternionSlerp(
     T=&RigVMFunction_MathFloatRemap(SourceMaximum=1, SourceMinimum=-1, Value=@n5.Result).Result
     B=&RigVMFunction_MathQuaternionMul(B=(X=0,Y=0,Z=-0.1305,W=0.9914), A=@n6.Transform.Rotation) {template="Multiply::Execute(in A,in B,out Result)"}.Result
     A=@n6.Transform.Rotation
    ) {template="Interpolate::Execute(in A,in B,in T,out Result)"}.Result
   )
   n7 = &RigVMFunction_MathFloatSin(Value=&RigVMFunction_Timeline(Speed=&RigVMFunction_RandomFloat(Duration=1, Maximum=5, Minimum=3, Seed=100) {template="Random::Execute(in Seed,in Minimum,in Maximum,in Duration,out Result)"}.Result).Time) {template="Sin::Execute(in Value,out Result)"}
   n8 = &RigUnit_GetTransform(Item=(Type=Bone,Name="dyn_l_tooth"), Space=LocalSpace)
   > &RigUnit_SetRotation(
    Item=(Type=Bone,Name="dyn_l_tooth")
    Weight=1
    Value=&RigVMFunction_MathQuaternionSlerp(
     T=&RigVMFunction_MathFloatRemap(SourceMaximum=1, SourceMinimum=-1, Value=@n7.Result).Result
     B=&RigVMFunction_MathQuaternionMul(B=(X=0,Y=0,Z=0.0349,W=0.9994), A=@n8.Transform.Rotation) {template="Multiply::Execute(in A,in B,out Result)"}.Result
     A=&RigVMFunction_MathQuaternionMul(B=(X=0,Y=0,Z=-0.0872,W=0.9962), A=@n8.Transform.Rotation) {template="Multiply::Execute(in A,in B,out Result)"}.Result
    ) {template="Interpolate::Execute(in A,in B,in T,out Result)"}.Result
   )
   n9 = &RigVMFunction_MathFloatSin(Value=&RigVMFunction_Timeline(Speed=&RigVMFunction_RandomFloat(Duration=1, Maximum=5, Minimum=3, Seed=100) {template="Random::Execute(in Seed,in Minimum,in Maximum,in Duration,out Result)"}.Result).Time) {template="Sin::Execute(in Value,out Result)"}
   n10 = &RigUnit_GetTransform(Item=(Type=Bone,Name="dyn_r_tooth"), Space=LocalSpace)
   > &RigUnit_SetRotation(
    Item=(Type=Bone,Name="dyn_r_tooth")
    Weight=1
    Value=&RigVMFunction_MathQuaternionSlerp(
     T=&RigVMFunction_MathFloatRemap(SourceMaximum=1, SourceMinimum=-1, Value=@n9.Result).Result
     B=&RigVMFunction_MathQuaternionMul(B=(X=0,Y=0,Z=-0.0436,W=0.999), A=@n10.Transform.Rotation) {template="Multiply::Execute(in A,in B,out Result)"}.Result
     A=&RigVMFunction_MathQuaternionMul(B=(X=0,Y=0,Z=0.0872,W=0.9962), A=@n10.Transform.Rotation) {template="Multiply::Execute(in A,in B,out Result)"}.Result
    ) {template="Interpolate::Execute(in A,in B,in T,out Result)"}.Result
   )
   n11 = &RigVMFunction_MathFloatSin(Value=&RigVMFunction_Timeline(Speed=&RigVMFunction_RandomFloat(Duration=1, Maximum=1.8, Minimum=1, Seed=100) {template="Random::Execute(in Seed,in Minimum,in Maximum,in Duration,out Result)"}.Result).Time) {template="Sin::Execute(in Value,out Result)"}
   n12 = &RigUnit_GetTransform(Item=(Type=Bone,Name="dyn_l_ear_01"), Space=LocalSpace)
   > &RigUnit_SetRotation(
    Item=(Type=Bone,Name="dyn_l_ear_01")
    Weight=1
    Value=&RigVMFunction_MathQuaternionSlerp(
     T=&RigVMFunction_MathFloatRemap(SourceMaximum=1, SourceMinimum=-1, Value=@n11.Result).Result
     A=&RigVMFunction_MathQuaternionMul(B=(X=0.26,Y=0.0026,Z=-0.0909,W=0.9613), A=@n12.Transform.Rotation) {template="Multiply::Execute(in A,in B,out Result)"}.Result
     B=&RigVMFunction_MathQuaternionMul(B=(X=-0.0428,Y=-0.0443,Z=0.0155,W=0.998), A=@n12.Transform.Rotation) {template="Multiply::Execute(in A,in B,out Result)"}.Result
    ) {template="Interpolate::Execute(in A,in B,in T,out Result)"}.Result
   )
   n13 = &RigVMFunction_MathFloatSin(Value=&RigVMFunction_Timeline(Speed=&RigVMFunction_RandomFloat(Duration=1, Maximum=2, Minimum=1, Seed=100) {template="Random::Execute(in Seed,in Minimum,in Maximum,in Duration,out Result)"}.Result).Time) {template="Sin::Execute(in Value,out Result)"}
   n14 = &RigUnit_GetTransform(Item=(Type=Bone,Name="dyn_r_ear_01"), Space=LocalSpace)
   > &RigUnit_SetRotation(
    Item=(Type=Bone,Name="dyn_r_ear_01")
    Weight=1
    Value=&RigVMFunction_MathQuaternionSlerp(
     T=&RigVMFunction_MathFloatRemap(SourceMaximum=1, SourceMinimum=-1, Value=@n13.Result).Result
     B=&RigVMFunction_MathQuaternionMul(B=(X=0.048,Y=0.0483,Z=-0.0892,W=0.9937), A=@n14.Transform.Rotation) {template="Multiply::Execute(in A,in B,out Result)"}.Result
     A=&RigVMFunction_MathQuaternionMul(B=(X=-0.0424,Y=0.028,Z=-0.0424,W=0.9978), A=@n14.Transform.Rotation) {template="Multiply::Execute(in A,in B,out Result)"}.Result
    ) {template="Interpolate::Execute(in A,in B,in T,out Result)"}.Result
   )
  ]}
]
hierarchy [
 root {loc=<0 0 0>, rot=<0 0 0 1>}
 pack {parent=root, loc=<0 -0.7316 0>, rot=<0 0 0 1>}
 dyn_body_01 {parent=pack, loc=<2.383 -9.905 1.3595>, rot=<-0.2601 -0.8567 0.1165 -0.4299>}
 dyn_body_02 {parent=dyn_body_01, loc=<10.2842 0 0>, rot=<0 0.5829 0 -0.8126>}
 dyn_body_03 {parent=dyn_body_02, loc=<8.7307 0 0>, rot=<0 0 0 1>}
 dyn_r_foot {parent=dyn_body_03, loc=<3.4492 0.8191 4.7669>, rot=<0.1635 0.5994 -0.1904 -0.7601>}
 dyn_l_foot {parent=dyn_body_03, loc=<3.365 -2.4249 4.6731>, rot=<-0.1144 0.6122 0.1835 -0.7606>}
 dyn_r_leg_01 {parent=dyn_body_02, loc=<6.1035 1.8417 2.9455>, rot=<0.0969 0.3176 -0.1025 -0.9377>}
 dyn_r_leg_02 {parent=dyn_r_leg_01, loc=<1.4303 0 0>, rot=<-0.0678 0.0667 0.3428 -0.9346>}
 dyn_l_leg_01 {parent=dyn_body_02, loc=<6.0147 -2.5029 2.7784>, rot=<0.002 0.323 0.1214 -0.9386>}
 dyn_l_leg_02 {parent=dyn_l_leg_01, loc=<1.4633 -0.0525 -0.0168>, rot=<-0.0084 0.0724 -0.2812 -0.9569>}
 dyn_head {parent=dyn_body_01, loc=<-0.4166 0 0>, rot=<0 0 1 0>}
 dyn_l_ear_01 {parent=dyn_head, loc=<8.5014 2.284 -5.3811>, rot=<-0.4838 -0.0476 0.8733 -0.0328>}
 dyn_l_ear_02 {parent=dyn_l_ear_01, loc=<5.8546 0 0>, rot=<0 0 0 1>}
 dyn_l_ear_03 {parent=dyn_l_ear_02, loc=<7.4486 0 0>, rot=<0 0 0 1>}
 dyn_r_ear_01 {parent=dyn_head, loc=<8.3661 -2.4938 -5.2547>, rot=<-0.4574 0.0389 0.8884 -0.0005>}
 dyn_r_ear_02 {parent=dyn_r_ear_01, loc=<5.9252 0 0>, rot=<0 0 0 1>}
 dyn_r_ear_03 {parent=dyn_r_ear_02, loc=<7.8271 0 0>, rot=<0 -0.1036 0 -0.9946>}
 dyn_l_tooth {parent=dyn_head, loc=<3.8355 2.3482 10.2729>, rot=<0.5753 -0.0374 0.8141 -0.0697>}
 dyn_r_tooth {parent=dyn_head, loc=<3.743 -1.8574 10.3937>, rot=<0.571 0.0799 0.8113 0.0967>}
 dyn_r_arm_01 {parent=dyn_body_01, loc=<2.6143 3.4997 5.6817>, rot=<0.1604 0.7244 0.1212 -0.6594>}
 dyn_r_arm_02 {parent=dyn_r_arm_01, loc=<2.5046 0 0>, rot=<0 -0.4528 0 -0.8916>}
 dyn_r_arm_03 {parent=dyn_r_arm_02, loc=<2.2885 0 0>, rot=<0 -0.3306 0 -0.9438>}
 dyn_l_arm_01 {parent=dyn_body_01, loc=<2.4031 -4.0361 5.6368>, rot=<-0.1579 0.725 -0.0945 -0.6637>}
 dyn_l_arm_02 {parent=dyn_l_arm_01, loc=<2.5046 0 0>, rot=<0 -0.4528 0 -0.8916>}
 dyn_l_arm_03 {parent=dyn_l_arm_02, loc=<2.2885 0 0>, rot=<0 -0.3306 0 -0.9438>}
]
