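# Physics asset for the Astral backpack cosmetic, defining collision bodies and joints for its dangling-ear secondary motion.
# Capsules cover the body and head, and each ear is a three-bone chain of thin boxes.
# Constraints link each ear root to the head and each segment to the previous one, with nearly locked twist and mirrored, asymmetric swing limits that are loosest at the tips.
# No constraint joins head and body.
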PhysicsAsset /BRCosmetics/Accessories/FORT_Backpacks/Backpack_F_MED_Astral/Meshes/F_MED_Astral_Pack_Physics
bodies {
 dyn_body_01 [{shapeType=Capsule, length=10.271, radius=8.7941, center=<5.5586 -0.2294 4.1583>, rotation=<rot 2.2143 -89.9978 14.5803>}]
 dyn_head [{shapeType=Capsule, length=2.4191, radius=8.5641, center=<4.1416 0.3295 2.4272>, rotation=<rot 0.7533 -89.9985 12.0552>}]
 dyn_l_ear_01 [{
   shapeType=Box
   extentX=2.3528
   extentY=0.5578
   extentZ=5.0237
   shapeName=dyn_l_ear_01_box
   center=<3.2322 0.0393 -0.0598>
   rotation=<rot -0.0224 89.6653 -90.2505>
  }]
 dyn_l_ear_02 [{
   shapeType=Box
   extentX=2.1088
   extentY=0.5
   extentZ=6.975
   shapeName=dyn_l_ear_02_box
   center=<3.5822 -0.0047 -0.0205>
   rotation=<rot -0.0011 90.2833 -89.6156>
  }]
 dyn_l_ear_03 [{
   shapeType=Box
   extentX=1.9711
   extentY=0.5
   extentZ=5.09
   shapeName=dyn_l_ear_03_box
   center=<2.1079 0.0377 0.0325>
   rotation=<rot 0.0348 91.1106 -89.3612>
  }]
 dyn_r_ear_01 [{
   shapeType=Box
   extentX=2.4367
   extentY=0.5
   extentZ=5.348
   shapeName=dyn_r_ear_01_box
   center=<3.124 -0.0938 -0.0755>
   rotation=<rot -1.134 91.069 -90.7825>
  }]
 dyn_r_ear_02 [{
   shapeType=Box
   extentX=2.1708
   extentY=0.5
   extentZ=6.9891
   shapeName=dyn_r_ear_02_box
   center=<3.615 -0.0823 -0.0114>
   rotation=<rot 0.0092 89.244 -88.242>
  }]
 dyn_r_ear_03 [{
   shapeType=Box
   extentX=1.8236
   extentY=0.5
   extentZ=4.8481
   shapeName=dyn_r_ear_03_box
   center=<2.0664 -0.1607 0.3574>
   rotation=<rot -0.9891 89.0872 -82.2356>
  }]
}
constraints [
 {bone1Name=dyn_l_ear_01, bone2Name=dyn_head, swing1LimitDegrees=5, swing2LimitDegrees=20, twistLimitDegrees=1}
 {bone1Name=dyn_l_ear_02, bone2Name=dyn_l_ear_01, swing1LimitDegrees=4, swing2LimitDegrees=30, twistLimitDegrees=1}
 {bone1Name=dyn_l_ear_03, bone2Name=dyn_l_ear_02, swing1LimitDegrees=15, swing2LimitDegrees=30, twistLimitDegrees=1}
 {bone1Name=dyn_r_ear_01, bone2Name=dyn_head, swing1LimitDegrees=5, swing2LimitDegrees=20, twistLimitDegrees=1}
 {bone1Name=dyn_r_ear_02, bone2Name=dyn_r_ear_01, swing1LimitDegrees=4, swing2LimitDegrees=30, twistLimitDegrees=1}
 {bone1Name=dyn_r_ear_03, bone2Name=dyn_r_ear_02, swing1LimitDegrees=15, swing2LimitDegrees=30, twistLimitDegrees=1}
]
