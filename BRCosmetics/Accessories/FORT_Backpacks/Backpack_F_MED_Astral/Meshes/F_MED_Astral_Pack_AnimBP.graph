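# Animation Blueprint for the Astral backpack cosmetic, adding physics and a sprint-driven ear flap to the pack.
# A RigidBody node simulates in root-bone space, then additive pitch offsets wave three bones per ear while sprinting and not crouching, before a Control Rig pass.
# The update event computes those offsets as real-time sine waves, with slightly different left and right frequencies and zero amplitude on both first ear bones.
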AnimBlueprint /BRCosmetics/Accessories/FORT_Backpacks/Backpack_F_MED_Astral/Meshes/F_MED_Astral_Pack_AnimBP
target_skeleton=&/F_MED_Astral_Pack_Skeleton
graphs [
 AnimGraph {nodes [
   n1 = RigidBody(ComponentPose=Local To Component(LocalPose=Input Pose())) {
    bOverrideWorldGravity=true
    overridePhysicsAsset=&/F_MED_Astral_Pack_Physics
    simulationSpace=BaseBoneSpace
    alphaBoolBlend {blendOption=Linear}
    baseBoneRef {boneName=root}
    componentLinearAccScale=<1 1 1>
    simSpaceSettings {worldAlpha=1, externalLinearVelocity {y=500, z=500}}
   }
   n2 = AND Boolean(A=Get bIsSprinting(), B=NOT Boolean(A=Get bIsCrouching()))
   n3 = "Transform (Modify) Bone - Bone: dyn_l_ear_01"!(
    ComponentPose=@n1
    bAlphaBoolEnabled:bool=@n2
    Translation="0,0,0"
    Rotation_Roll:float=0
    Rotation_Pitch:float=Get wave_01() -> (wave_01:double)
    Rotation_Yaw:float=0
    Scale="1,1,1"
   ) -> (Pose) {
    alphaInputType=Bool
    rotationMode=BMM_Additive
    rotationSpace=BCS_ParentBoneSpace
    alphaBoolBlend {blendOption=Linear}
    boneToModify {boneName=dyn_l_ear_01}
   }
   n4 = "Transform (Modify) Bone - Bone: dyn_l_ear_02"!(
    ComponentPose=@n3
    bAlphaBoolEnabled:bool=@n2
    Translation="0,0,0"
    Rotation_Roll:float=0
    Rotation_Pitch:float=Get wave_02() -> (wave_02:double)
    Rotation_Yaw:float=0
    Scale="1,1,1"
   ) -> (Pose) {
    alphaInputType=Bool
    rotationMode=BMM_Additive
    rotationSpace=BCS_ParentBoneSpace
    alphaBoolBlend {blendOption=Linear}
    boneToModify {boneName=dyn_l_ear_02}
   }
   n5 = "Transform (Modify) Bone - Bone: dyn_l_ear_03"!(
    ComponentPose=@n4
    bAlphaBoolEnabled:bool=@n2
    Translation="0,0,0"
    Rotation_Roll:float=0
    Rotation_Pitch:float=Get wave_03() -> (wave_03:double)
    Rotation_Yaw:float=0
    Scale="1,1,1"
   ) -> (Pose) {
    alphaInputType=Bool
    rotationMode=BMM_Additive
    rotationSpace=BCS_ParentBoneSpace
    alphaBoolBlend {blendOption=Linear}
    boneToModify {boneName=dyn_l_ear_03}
   }
   n6 = "Transform (Modify) Bone - Bone: dyn_r_ear_01"!(
    ComponentPose=@n5
    bAlphaBoolEnabled:bool=@n2
    Translation="0,0,0"
    Rotation_Roll:float=0
    Rotation_Pitch:float=Get r_wave_01() -> (r_wave_01:double)
    Rotation_Yaw:float=0
    Scale="1,1,1"
   ) -> (Pose) {
    alphaInputType=Bool
    rotationMode=BMM_Additive
    rotationSpace=BCS_ParentBoneSpace
    alphaBoolBlend {blendOption=Linear}
    boneToModify {boneName=dyn_r_ear_01}
   }
   n7 = "Transform (Modify) Bone - Bone: dyn_r_ear_02"!(
    ComponentPose=@n6
    bAlphaBoolEnabled:bool=@n2
    Translation="0,0,0"
    Rotation_Roll:float=0
    Rotation_Pitch:float=Get r_wave_02() -> (r_wave_02:double)
    Rotation_Yaw:float=0
    Scale="1,1,1"
   ) -> (Pose) {
    alphaInputType=Bool
    rotationMode=BMM_Additive
    rotationSpace=BCS_ParentBoneSpace
    alphaBoolBlend {blendOption=Linear}
    boneToModify {boneName=dyn_r_ear_02}
   }
   n8 = "Transform (Modify) Bone - Bone: dyn_r_ear_03"!(
    ComponentPose=@n7
    bAlphaBoolEnabled:bool=@n2
    Translation="0,0,0"
    Rotation_Roll:float=0
    Rotation_Pitch:float=Get r_wave_03() -> (r_wave_03:double)
    Rotation_Yaw:float=0
    Scale="1,1,1"
   ) -> (Pose) {
    alphaInputType=Bool
    rotationMode=BMM_Additive
    rotationSpace=BCS_ParentBoneSpace
    alphaBoolBlend {blendOption=Linear}
    boneToModify {boneName=dyn_r_ear_03}
   }
   n9 = Control Rig(Source=Component To Local(ComponentPose=@n8)) {
    alphaBoolBlend {blendOption=Linear}
    controlRigAssetReference {blueprintRigClass=&/F_MED_Astral_Pack_CtrlRig.F_MED_Astral_Pack_CtrlRig_C}
   }
   Output Pose(Result=@n9) {layerGroup=None}
  ]}
 EventGraph {nodes [
   Event BlueprintUpdateAnimation()
   n1 = GetRealTimeSeconds()
   n2 = float * float(A=@n1, B=8)
   > Set wave_01(wave_01:double=float * float(A="Sin (Radians)"(A=@n2), B=0)) -> (Output_Get:double)
   n3 = float * float(A=@n1, B=8)
   > Set wave_02(wave_02:double=float * float(A="Sin (Radians)"(A=@n3), B=25)) -> (Output_Get:double)
   n4 = float * float(A=@n1, B=8)
   > Set wave_03(wave_03:double=float * float(A="Sin (Radians)"(A=@n4), B=20)) -> (Output_Get:double)
   n5 = float * float(A=@n1, B=8.02)
   > Set r_wave_01(r_wave_01:double=float * float(A="Sin (Radians)"(A=@n5), B=0)) -> (Output_Get:double)
   n6 = float * float(A=@n1, B=8.02)
   > Set r_wave_02(r_wave_02:double=float * float(A="Sin (Radians)"(A=@n6), B=25)) -> (Output_Get:double)
   n7 = float * float(A=@n1, B=8.02)
   > Set r_wave_03(r_wave_03:double=float * float(A="Sin (Radians)"(A=@n7), B=20)) -> (Output_Get:double)
   TryGetPawnOwner()
  ]}
]
variables [
 r_wave_01 {type=number}
 r_wave_02 {type=number}
 r_wave_03 {type=number}
 wave_01 {type=number}
 wave_02 {type=number}
 wave_03 {type=number}
]
